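# Physics asset for the hanging string bones on the F_MED_ProngCling female medium character body, simulating a left and right string.
# Each string bone carries two overlapping capsules and is constrained to its init bone with a narrow swing cone and zero twist limit.
# Pelvis and both thigh capsules supply body shapes around the strings, while tiny capsules mark the master, base and init string bones.
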PhysicsAsset /BRCosmetics/Characters/Player/Female/Medium/Bodies/F_MED_ProngCling/Meshes/F_MED_ProngCling_Physics_String
bodies {
 base_dyn_string_00 [{shapeType=Capsule, length=0.505, radius=0.505, shapeName=base_dyn_string_00_capsule}]
 dyn_ini_string_l_00 [{shapeType=Capsule, length=0.505, radius=0.505, shapeName=dyn_ini_string_l_00_capsule}]
 dyn_ini_string_r_00 [{shapeType=Capsule, length=0.505, radius=0.505, shapeName=dyn_ini_string_r_00_capsule}]
 dyn_string_l_00 [
  {shapeType=Capsule, length=2.88, radius=1, shapeName=dyn_string_l_00_capsule, center=<2.5 0 0>, rotation=<rot 0.7941 -180 180>}
  {shapeType=Capsule, length=2.88, radius=1, shapeName=dyn_string_l_00_capsule, center=<0.5 0 0>, rotation=<rot 0.7941 -180 180>}
 ]
 dyn_string_r_00 [
  {shapeType=Capsule, length=2.8, radius=1, shapeName=dyn_string_r_00_capsule, center=<2.5 0 0>, rotation=<rot 0.7941 -180 -180>}
  {shapeType=Capsule, length=2.8, radius=1, shapeName=dyn_string_r_00_capsule, center=<0.5 0 0>, rotation=<rot 0.7941 -180 -180>}
 ]
 master_dyn_string_0 [{shapeType=Capsule, length=0.505, radius=0.505, shapeName=master_dyn_string_0_capsule}]
 pelvis [{shapeType=Capsule, length=19.9067, radius=13.9913, shapeName=pelvis_capsule, center=<-7.0239 -3.3127 0>}]
 thigh_l [{
   shapeType=Capsule
   length=44.9934
   radius=10.2109
   shapeName=thigh_l_capsule
   center=<-28.7416 -2.0119 -0.9389>
   rotation=<rot 86.2835 179.9998 179.9999>
  }]
 thigh_r [{
   shapeType=Capsule
   length=44.9934
   radius=10.2109
   shapeName=thigh_l_capsule
   center=<28.7416 2.0119 -0.9389>
   rotation=<rot 86.2835 179.9998 179.9999>
  }]
}
constraints [
 {bone1Name=dyn_string_r_00, bone2Name=dyn_ini_string_r_00, swing1LimitDegrees=15, swing2LimitDegrees=15, twistLimitDegrees=0}
 {bone1Name=dyn_string_l_00, bone2Name=dyn_ini_string_l_00, swing1LimitDegrees=15, swing2LimitDegrees=15, twistLimitDegrees=0}
]
