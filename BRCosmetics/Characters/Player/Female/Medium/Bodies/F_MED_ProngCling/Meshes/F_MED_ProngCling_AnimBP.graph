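# Animation Blueprint for the female ProngCling body cosmetic, adding secondary motion and a backpack-dependent material toggle.
# It copies the parent mesh pose through a Control Rig fed crouch and targeting states, then simulates a RigidBody string physics asset in pelvis space.
# Updates compute a gravity override and lower a skirt physics alpha during vigorous movement or emotes.
# Customization load and weapon changes toggle an element on body and outline materials by backpack tag.
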AnimBlueprint /BRCosmetics/Characters/Player/Female/Medium/Bodies/F_MED_ProngCling/Meshes/F_MED_ProngCling_AnimBP
target_skeleton=&/F_MED_ProngCling_Skeleton
graphs [
 AnimGraph {nodes [
   n1 = Control Rig!(Source=Copy Pose From Mesh() {bUseAttachedParent=true}, Alpha, IsCrouching:bool=Get bIsCrouching(), IsTargeting:bool=Get bIsTargeting()) -> (Pose) {controlRigAssetReference {blueprintRigClass=&/F_MED_ProngCling_CtrlRig.F_MED_ProngCling_CtrlRig_C}}
   n2 = RigidBody!(ComponentPose=Local To Component(LocalPose=@n1), Alpha, OverrideWorldGravity:Vector="0,0,-980", ExternalForce) -> (Pose) {
    bClampLinearTranslationLimitToRefPose=true
    bOverrideWorldGravity=true
    overridePhysicsAsset=&/F_MED_ProngCling_Physics_String
    simulationSpace=BaseBoneSpace
    alphaScaleBiasClamp {bInterpResult=true}
    baseBoneRef {boneName=pelvis}
    overrideWorldGravity {z=-980}
    simSpaceSettings {dampingAlpha=0, maxAngularAcceleration=40, maxAngularVelocity=8, maxLinearAcceleration=3500, maxLinearVelocity=700, worldAlpha=1}
   }
   Output Pose(Result=Component To Local(ComponentPose=@n2)) {layerGroup=None}
  ]}
 EventGraph {nodes [
   n1 = Event OnCharacterCustomizationLoaded()
   n2 = GetPartSkeletalMeshComponent()
   n3 = Get Dynamic Material Instance(execute=[@n1.then, @n14.then], SkeletalMeshComponent=@n2)
   > n4 = GetOrCreateDynamicOverlayMaterial(MeshComponent=@n2, MaterialName=MI_ProngCling_Body_InvHullOutline)
   > n5 = Branch(Condition=BackpackItemDefinitionHasAnyMetaTag!(Tags:Array of Gameplay Tag Structures=Make Array("[0]"=(TagName="Cosmetics.MetaTags.BackpackPuzzleShed")), bRequiresEquipped:bool=false) -> (ReturnValue:bool))
   > SetScalarParameterValue(self=[@n4, @n3], ParameterName=Hide Element 01, Value=0)
   SetScalarParameterValue(execute=@n5.else, self=[@n4, @n3], ParameterName=Hide Element 01, Value=1)
   Event BlueprintUpdateAnimation()
   > n6 = "Gravity Override State Function (Parameter Asset) (Nativized)"(
    CharacterPartAnimInstance=Self-Reference()
    GameTimeInSeconds=GetGameTimeInSeconds()
    Parameters=&/Fortnite_ProngCling_FaceAcc_Parameters
   )
   > Set GravityOverride(GravityOverride=@n6)
   n7 = Get bIsTacticalSprinting()
   n8 = Get bIsFalling()
   n9 = Get bIsJumping()
   > n10 = Branch(Condition=OR Boolean!(
     A=AND Boolean(A=@n9, B=@n7)
     B=AND Boolean(A=@n7, B=@n8)
     C:bool=Get bIsTacticalSprinting()
     D:bool=Get bIsKneeSliding()
     E:bool=Get bIsPlayingEmote()
    ) -> (ReturnValue))
   > Set alphaSkirtRigidBody float(alphaSkirtRigidBody float:double=0.6) -> (Output_Get:double)
   n11 = NOT Boolean(A=Get bIsCrouching())
   n12 = Get bIsSprinting()
   n13 = Branch(execute=@n10.else, Condition=OR Boolean!(A=Get bIsJumping(), B=Get bIsFalling(), C=AND Boolean(A=@n12, B=@n11)) -> (ReturnValue))
   > Set alphaSkirtRigidBody float(alphaSkirtRigidBody float=0.6)
   Set alphaSkirtRigidBody float(execute=@n13.else, alphaSkirtRigidBody float=1)
   n14 = Event OnWeaponChanged()
  ]}
]
variables [
 alphaSkirtRigidBody float {type=number, default=1}
 GravityOverride {type=Vector}
]
