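# Physics asset for the face accessory part of the female medium ProngCling character, defining capsule bodies and joint limits for its dynamic hair bones.
# A three-bone right hair chain swings moderately without twist, while single center and left bones hang from their own base bones with tight, nearly rigid limits.
# Unconstrained capsules on head, neck, upper spine, right clavicle and right upper arm provide body volumes around the hair.
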PhysicsAsset /BRCosmetics/Characters/Player/Female/Medium/Bodies/F_MED_ProngCling/Meshes/Parts/F_MED_ProngCling_FaceAcc_Physics
bodies {
 base_dyn_hair_c_00 [{shapeType=Capsule, length=0.505, radius=0.505, shapeName=base_dyn_hair_c_00_capsule, rotation=<rot 0 -90 0>}]
 base_dyn_hair_l_00 [{shapeType=Capsule, length=0.505, radius=0.505, shapeName=base_dyn_hair_l_00_capsule, rotation=<rot 0 -90 0>}]
 base_dyn_hair_r_00 [{shapeType=Capsule, length=0.505, radius=0.505, shapeName=base_dyn_hair_r_00_capsule, rotation=<rot 0 -90 0>}]
 clavicle_r [{
   shapeType=Capsule
   length=15.1671
   radius=9.4675
   shapeName=clavicle_r_capsule
   center=<-2.3919 1.1506 4.0198>
   rotation=<rot -0.0001 -90 -92.3252>
  }]
 dyn_hair_c_00 [{
   shapeType=Capsule
   length=0.7929
   radius=1.1722
   shapeName=dyn_hair_c_00_capsule
   center=<5.409 -1.194 0.4345>
   rotation=<rot 0 74.9358 -62.6075>
  }]
 dyn_hair_l_00 [{
   shapeType=Capsule
   length=2.7678
   radius=1.2524
   shapeName=dyn_hair_l_00_capsule
   center=<6.8958 -0.4013 -0.5439>
   rotation=<rot 0 -111.4886 -74.995>
  }]
 dyn_hair_r_00 [
  {
   shapeType=Capsule
   length=6.6789
   radius=2.7989
   shapeName=dyn_hair_r_00_capsule
   center=<1.6935 -0.0822 0.046>
   rotation=<rot -1.1895 178.5009 -87.1615>
  }
  {
   shapeType=Capsule
   length=6.6672
   radius=2.7989
   shapeName=dyn_hair_r_00_capsule
   center=<-3.1452 0.0633 0.1155>
   rotation=<rot -1.1895 178.5009 -87.1615>
  }
 ]
 dyn_hair_r_01 [{
   shapeType=Capsule
   length=4.5491
   radius=2.8757
   shapeName=dyn_hair_r_01_capsule
   center=<1.4564 -0.3668 0.4926>
   rotation=<rot 0 178.6603 -80.6193>
  }]
 dyn_hair_r_02 [{
   shapeType=Capsule
   length=7.4211
   radius=3.0193
   shapeName=dyn_hair_r_02_capsule
   center=<3.9857 0.0782 -0.3543>
   rotation=<rot 12.2424 111.7521 -106.7606>
  }]
 head [{shapeType=Capsule, length=1.4753, radius=17.2102, center=<9.3437 0 0>}]
 neck_01 [{shapeType=Capsule, length=0, radius=6.5, shapeName=neck_01_capsule, rotation=<rot 0 -90 0>}]
 spine_05 [{shapeType=Capsule, length=11.3049, radius=15.0337, shapeName=spine_05_capsule, center=<-9.3633 0.693 0>}]
 upperarm_r [{
   shapeType=Capsule
   length=27.0368
   radius=7.7591
   shapeName=upperarm_r_capsule
   center=<-11.4755 0.7978 -0.2192>
   rotation=<rot 0.0122 -89.5367 -91.539>
  }]
}
constraints [
 {bone1Name=dyn_hair_r_00, bone2Name=base_dyn_hair_r_00, swing1LimitDegrees=20, swing2LimitDegrees=15, twistLimitDegrees=0}
 {bone1Name=dyn_hair_r_01, bone2Name=dyn_hair_r_00, swing1LimitDegrees=20, swing2LimitDegrees=15, twistLimitDegrees=0}
 {bone1Name=dyn_hair_r_02, bone2Name=dyn_hair_r_01, swing1LimitDegrees=20, swing2LimitDegrees=15, twistLimitDegrees=0}
 {bone1Name=dyn_hair_c_00, bone2Name=base_dyn_hair_c_00, swing1LimitDegrees=5, swing2LimitDegrees=5, twistLimitDegrees=5}
 {bone1Name=dyn_hair_l_00, bone2Name=base_dyn_hair_l_00, swing1LimitDegrees=0, swing2LimitDegrees=0.5, twistLimitDegrees=0}
]
