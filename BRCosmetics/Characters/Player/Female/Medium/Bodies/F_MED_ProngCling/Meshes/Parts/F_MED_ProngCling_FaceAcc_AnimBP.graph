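# Animation Blueprint for the ProngCling character's face accessory, adding hair dynamics on top of the copied character pose.
# It blends parent and body-mesh poses at the upper arms, then runs Control Rig, spine-relative RigidBody simulation and an additive right-hair bone offset.
# A front-end idle curve blends hair bones from the cached dynamics to the montage slot.
# Each update computes a parameter-driven gravity override and an inverted pistol-aim curve.
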AnimBlueprint /BRCosmetics/Characters/Player/Female/Medium/Bodies/F_MED_ProngCling/Meshes/Parts/F_MED_ProngCling_FaceAcc_AnimBP
target_skeleton=&/F_MED_ProngCling_FaceAcc_Skeleton
graphs [
 AnimGraph {nodes [
   n1 = Layered blend per bone(
    BasePose=Use cached pose 'dynamicsCache'()
    BlendPoses_0=Slot 'DefaultSlot'()
    BlendWeights_0=Get Custom Idle Is Playing Float() -> (Custom Idle Is Playing Float:double)
   ) {layerSetup [{branchFilters [
       {boneName=base_dyn_hair_c_00}
       {boneName=dyn_hair_c_00}
       {boneName=base_dyn_hair_l_00}
       {boneName=dyn_hair_l_00}
       {boneName=base_dyn_hair_r_00}
       {boneName=dyn_hair_r_00}
       {boneName=dyn_hair_r_01}
       {boneName=dyn_hair_r_02}
      ]}]}
   Output Pose(Result=@n1) {layerGroup=None}
   n2 = Get Body_Mesh()
   n3 = Control Rig(Source=Layered blend per bone(BasePose=Copy Pose From Mesh() {bUseAttachedParent=true}, BlendPoses_0=Copy Pose From Mesh(SourceMeshComponent=@n2)) {layerSetup [{branchFilters [{boneName=upperarm_l}, {boneName=upperarm_r}]}]}) {controlRigAssetReference {blueprintRigClass=&/F_MED_ProngCling_FaceAcc_CtrlRig.F_MED_ProngCling_FaceAcc_CtrlRig_C}}
   n4 = RigidBody!(ComponentPose=Local To Component(LocalPose=@n3), Alpha, OverrideWorldGravity:Vector="0,0,0", ExternalForce) -> (Pose) {
    bClampLinearTranslationLimitToRefPose=true
    bOverrideWorldGravity=true
    overridePhysicsAsset=&/F_MED_ProngCling_FaceAcc_Physics
    simulationSpace=BaseBoneSpace
    baseBoneRef {boneName=spine_05}
    simSpaceSettings {dampingAlpha=0.5, maxAngularAcceleration=40, maxAngularVelocity=8, maxLinearAcceleration=3500, maxLinearVelocity=700, worldAlpha=1}
   }
   n5 = "Transform (Modify) Bone - Bone: dyn_hair_r_01"!(ComponentPose=@n4, Alpha=1, Translation="0,-2,0", Rotation="22.5,0,0") -> (Pose) {
    rotationMode=BMM_Additive
    rotationSpace=BCS_BoneSpace
    translationMode=BMM_Additive
    translationSpace=BCS_BoneSpace
    alphaScaleBiasClamp {bInterpResult=true}
    boneToModify {boneName=dyn_hair_r_01}
    rotation {pitch=22.5}
    translation {y=-2}
   }
   Save cached pose 'dynamicsCache'(Pose=Component To Local(ComponentPose=@n5)) {cacheName=dynamicsCache}
  ]}
 EventGraph {nodes [
   n1 = Event BlueprintUpdateAnimation()
   > n2 = "Gravity Override State Function (Parameter Asset) (Nativized)"(
    CharacterPartAnimInstance=Self-Reference()
    GameTimeInSeconds=@n1.DeltaTimeX
    Parameters=/BRCosmetics/Characters/Player/Female/Medium/Bodies/F_MED_ProngCling/Meshes/Fortnite_ProngCling_FaceAcc_Parameters
   )
   > Set Gravity Override(Gravity Override=@n2)
   > Set aiming up pistol(aiming up pistol:double=float - float(A=1, B=GetCurveValue(CurveName=aiming_up_pistol))) -> (Output_Get:double)
   > Branch(Condition=Get bIsInFrontEnd())
   > Set Custom Idle Is Playing Float(Custom Idle Is Playing Float:double=GetCurveValue(self=GetOwnerFrontEndAnimInstance(), CurveName=Frontend_Idle_Bypass)) -> (Output_Get:double)
   Event BlueprintInitializeAnimation()
   > Set Body_Mesh(Body_Mesh=GetOwnerSkeletalMeshComponent())
  ]}
]
variables [
 aiming up pistol {type=number}
 Body_Mesh {type=SkeletalMeshComponent}
 Custom Idle Is Playing Float {type=number}
 Gravity Override {type=Vector}
]
